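; Server-authoritative pawn for the TidalCrane ToyFox boss, a giant bug that is stunned at low health.
; Only AffectBoss-tagged damage builds frustration, and health clamps at the stun threshold, where a stun effect applies.
; The server gathers reposition actors through the global registry and times target proximity, with frustration meant to steer repositioning and melee.
; Clients watch stun tags to show piloting players a final-strike widget; custom primitive data drives stagger-denied visuals.
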(event EventFindAllNearbyRelocationActors
  (bind _returnvalue (GameplayTags|MakeLiteralGameplayTag "(TagName=\"TidalCrane.ToyFox.Boss.Ability.Reposition.Actor\")"))
  (bind _returnvalue_2 (WorldSubsystems|GetFortGlobalActorRegistrySubsystem))
  ;[0 0.15 0.05493 0.5] FindAllNearbyRelocationActors: Overridden to gather all the reposition actors from the world.
  (Networking|SwitchHasAuthority
    (:Authority
      (bind _outactors (Class|FortGlobalActorRegistrySubsystem|GetRegisteredActors _returnvalue_2 _returnvalue))
      (bind _output_get (Variables|InternalData|SetRepositionActors _outactors))
      (for _array_element _output_get
        (Utilities|Map|Add (Variables|InternalData|GetRepositionActorsAndBases) _array_element))
      (bind _returnvalue_3 (Class|FortGlobalActorRegistrySubsystem|GetActorRegistryDataforGameplayTag _returnvalue_2 _returnvalue))
      (Utilities|IsValid _returnvalue_3
        (:Is%20Valid
          (EventDispatchers|BindEventtoOnActorRegistered _returnvalue_3 (EventDispatchers|CreateEvent :!event Custom|HandleRepositionActorFound)))))))

(event Custom|HandleRepositionActorFound
  (params (GameplayTag :type Gameplay%20Tag%20Structure%20%28by%20ref%29) (RegisteredActor :type Actor%20Object%20Reference))
  ;[0 0.15 0.05493 0.5] FindAllNearbyRelocationActors: Overridden to gather all the reposition actors from the world.
  (Networking|SwitchHasAuthority
    (:Authority
      (Utilities|IsValid RegisteredActor
        (:Is%20Valid
          (if (not (Utilities|Array|ContainsItem (Variables|InternalData|GetRepositionActors) RegisteredActor))
            (Utilities|Array|Add (Variables|InternalData|GetRepositionActors) RegisteredActor)
            (Utilities|Map|Add (Variables|InternalData|GetRepositionActorsAndBases) RegisteredActor)))))))

(event EventBeginPlay
  ;[0 0.09896 0.01715 0.5] BeginPlay: Start a timer to track proximity to the target giant pawn.
  (|CallParentFunction)
  (Networking|SwitchHasAuthority
    (:Authority
      (Class|FortAsyncActionCurrentPlaylistReady|PlaylistDataReadyAsync
        (:OnReady
          (Utilities|Time|SetTimerbyEvent (EventDispatchers|CreateEvent :!event UpdateTargetProximityDuration) (ScalableFloat|GetCachedScalableFloatValue (Variables|Hotfix|GetHotfix_ProximityUpdateFreq)) true)
          (Utilities|IsValid (Ability|GetFortAbilitySystemComponent self)
            (:Is%20Valid
              (CallFunction|HandleAbilitySystemComponentInitialized))
            (:Is%20Not%20Valid
              (EventDispatchers|AssignOnAbilitySystemComponentInitialized (EventDispatchers|CreateEvent :!event Custom|HandleAbilitySystemComponentInitialized))))
          ;;[0.02377 0.1076 0.15 0.5] Editor specific debugging timer.
          (if (Editor|InEditororPIE)
            (Utilities|Time|SetTimerbyEvent (EventDispatchers|CreateEvent :!event EditorDebugTick) 1.0 true)
            (MaterialAnim|MStaggerImmunityStartupValues)))))
    (:Remote
      (CallFunction|ListenforLocalStunState))))

(event EventEndPlay
  (params (EndPlayReason :type EEndPlayReason%20Enum :default "Destroyed"))
  ; EndPlay
  (|CallParentFunction EndPlayReason)
  (Ability|Async|EndAction (Variables|UI|GetAsyncAction_ListenForStun)))

(event EventOnDamageServer
  (params (Damage :type Float%20%28single-precision%29) (DamageTags :type Gameplay%20Tag%20Container%20Structure%20%28by%20ref%29) (Momentum :type Vector :default "0, 0, 0") (HitInfo :type Hit%20Result%20Structure%20%28by%20ref%29) (InstigatedBy :type Controller%20Object%20Reference) (DamageCauser :type Actor%20Object%20Reference) (EffectContext :type Gameplay%20Effect%20Context%20Handle%20Structure))
  ;[0.002228 0.15 0.09275 0.5] OnDamageServer: Unit took damage, apply frustration and stun if needed.
  (|CallParentFunction Damage DamageTags Momentum HitInfo InstigatedBy DamageCauser EffectContext)
  (if (GameplayTags|HasTag DamageTags "(TagName=\"Gameplay.Damage.TidalCrane.ToyFox.AffectBoss\")" true)
    (Frustration|ApplyFrustration (Math|Float|SafeDivide Damage (* (ScalableFloat|GetCachedScalableFloatValue (Variables|Hotfix|GetHotfix_DamageTakenAsHealthPercentToMaxFrustration)) (Health|GetMaxHealth))) "OnDamaged")
    (if (<= (Health|GetHealth) (ScalableFloat|GetCachedScalableFloatValue (Variables|Hotfix|GetHotfix_StunnedHealthThreshold)))
      (CallFunction|ApplyStun))))

(event EventOnDeathPlayEffects
  (params (Damage :type Float%20%28single-precision%29) (DamageTags :type Gameplay%20Tag%20Container%20Structure%20%28by%20ref%29) (Momentum :type Vector :default "0, 0, 0") (HitInfo :type Hit%20Result%20Structure%20%28by%20ref%29) (InstigatedBy :type Fort%20Pawn%20Object%20Reference) (DamageCauser :type Actor%20Object%20Reference) (EffectContext :type Gameplay%20Effect%20Context%20Handle%20Structure)))

(event Custom|HandleAbilitySystemComponentInitialized
  (bind _hotfix_maxhealth (Variables|Hotfix|GetHotfix_MaxHealth))
  (bind _returnvalue (ScalableFloat|GetCachedScalableFloatValue _hotfix_maxhealth))
  ;[0 0.15 0.0877 0.5] HandleAbilitySystemComponentInitialized: Update the health of the bug to the desired value.
  (Networking|SwitchHasAuthority
    (:Authority
      (Health|SetMaxHealthDontCreateASC _returnvalue)
      (Health|SetHealthDontCreateASC _returnvalue)
      (EventDispatchers|UnbindEventfromOnAbilitySystemComponentInitialized (delegate Custom|HandleAbilitySystemComponentInitialized))
      (Fortnite|FortShippingLog self (Utilities|String|Append "HandleAbilitySystemComponentInitialized: Max health is set to " (Utilities|String|ToString(Float) (Health|GetHealth))) (not (Math|Float|NearlyEqual(Float) (Health|GetHealth) _returnvalue 0.1))))))

(event EventTidalCraneQueen_StartTakeOver)

(event EventToyFoxQueen_StartTakeOver
  ;[0.15 0.04181 0.009335 0.5] ToyFoxQueen_StartTakeOver: This is a cheat to help the cinematics team.
  (Networking|SwitchHasAuthority
    (:Authority
      (if (AI|Logic|IsRunning (Variables|Default|GetStateTree))
        (bind _object (Utilities|AsyncLoadAsset (Variables|Cheat|GetCheatTakeOverAbilitySet))
          (:Completed
            (bind _asfort_ability_set (Utilities|Casting|CastToFortAbilitySet _object))
            (Ability|EquipFortAbilitySet self _asfort_ability_set (Ability|GetFortAbilitySystemComponent self))))))))

(event EventToyFoxQueen_ApplyStun
  ;[0.002017 0.05813 0.15 0.5] ToyFoxQueen_ApplyStun: Implemented through BPI to help the cheat manager not pull in hard references.
  (Networking|SwitchHasAuthority
    (:Authority
      (CallFunction|ApplyStun))))

(event EventOnDamagePlayEffects
  (params (Damage :type Float%20%28single-precision%29) (DamageTags :type Gameplay%20Tag%20Container%20Structure%20%28by%20ref%29) (Momentum :type Vector :default "0, 0, 0") (HitInfo :type Hit%20Result%20Structure%20%28by%20ref%29) (InstigatedBy :type Fort%20Pawn%20Object%20Reference) (DamageCauser :type Actor%20Object%20Reference) (EffectContext :type Gameplay%20Effect%20Context%20Handle%20Structure))
  ;[0 0.09993 0.15 0.5] OnDamagePlayEffects: Handle FXs when the Giant Queen is damaged.
  (|CallParentFunction Damage DamageTags Momentum HitInfo InstigatedBy DamageCauser EffectContext)
  (if (and (GameplayTags|HasTag DamageTags "(TagName=\"Gameplay.Damage.TidalCrane.ToyFox.AffectBoss.AppliesStagger\")") (GameplayTags|HasMatchingGameplayTag self "(TagName=\"Ability.TidalCrane.ToyFox.StaggerImmunity\")"))
    (MaterialAnim|MStaggerDenied)))

(event EventSendQuestCustomVerb
  ;[0.08349 0 0.15 0.5] Overridden to bypass queen death quest custom verb.
  ; This quest is completed on the toyfox side instead.
  (Utilities|Array|Clear (Variables|Default|GetAllDamagingPawns))
  (|CallParentFunction))

(event Custom|ListenForLocalStunState
  (bind _enemystunnedgameplaytagquery (Variables|UI|GetEnemyStunnedGameplayTagQuery))
  ; Listen for stun state on the client to determine if we should show final strike widget
  (bind _asynctaskproxy (Ability|Tasks|WaitGameplayTagQueryOnActor 0 _enemystunnedgameplaytagquery)
    (:then
      (Variables|UI|SetAsyncAction_ListenForStun _asynctaskproxy)
      (bind _asynctaskproxy_2 (Ability|Tasks|WaitGameplayTagQueryOnActor 0 _enemystunnedgameplaytagquery "WhenFalse")
        (:then
          (Variables|UI|SetAsyncAction_ListenForNoStun _asynctaskproxy_2))
        (:Triggered
          (UI|ShowFinalStrikeWidget false))))
    (:Triggered
      (UI|ShowFinalStrikeWidget (UI|IsLocalPlayerPiloting)))))

(fn ApplyFrustration
  (params (Amount :type Float%20%28double-precision%29) (Reason :type String))
  ;[0.06879 0.05501 0.15 0.5] ApplyFrustration: Increment the frustration value of the bug.
  ; Frustration will be used to dictate gameplay behavior such as reposition as well as melee attack types.
  (Networking|SwitchHasAuthority
    (:Authority
      (bind _output_get (Variables|Transient|SetFrustration (Math|Float|Clamp(Float) (+ Amount (Frustration|GetFrustration)))))
      (Fortnite|FortShippingLog self (Utilities|String|Append "ApplyFrustration: Invoked with amount of " (Utilities|String|ToString(Float) Amount) " for reason " Reason (Utilities|String|ToString(Float) _output_get))))))

(fn ApplyStun
  (Networking|SwitchHasAuthority
    (:Authority
      (Fortnite|ApplyGameplayEffectsToActor self (Utilities|Array|MakeArray (Variables|Settings|GetStunnedGameplayEffectClass)) (0.0 0.0 0.0) false 0 (Ability|GetAbilitySystemComponent 0) (GameplayEffects|MakeEffectContext (Ability|GetAbilitySystemComponent 0))))))

(fn BlueprintModifyIncomingDamage
  (params (Damage :type Float%20%28single-precision%29) (DamageTags :type Gameplay%20Tag%20Container%20Structure%20%28by%20ref%29) (EffectContext :type Gameplay%20Effect%20Context%20Handle%20Structure%20%28by%20ref%29) (EventInstigator :type Controller%20Object%20Reference) (DamageCauser :type Actor%20Object%20Reference))
  (bind _hotfix_stunnedhealththreshold (Variables|Hotfix|GetHotfix_StunnedHealthThreshold))
  (bind _returnvalue (ScalableFloat|GetCachedScalableFloatValue _hotfix_stunnedhealththreshold))
  ;[0.004996 0.15 0.1238 0.5] BlueprintModifyIncomingDamage: Only allow damage from desired sources.
  (bind _returnvalue_2 (|CallParentFunction Damage DamageTags EffectContext EventInstigator DamageCauser))
  (Utilities|FlowControl|Sequence
    (:then_0
      ;;[0.01604 0.03761 0.15 0.5] Do not let the boss die to damage directly. It needs to enter threshold health value to go into stunned state.
      (if (< (- (Health|GetHealth) Damage) _returnvalue)
        (return (Math|Float|Max(Float) (- (Health|GetHealth) _returnvalue)))))
    (:then_1
      (if (GameplayTags|HasTag DamageTags "(TagName=\"Gameplay.Damage.TidalCrane.ToyFox.AffectBoss\")")
        (return _returnvalue_2)
        (else
          (return 0.0))))))

(fn DebugToyFoxBossAttackNearestPlayer
  (params (bEnable :type Boolean))
  ;[0 0 0 0.5] Debug cheat to have the boss attack the nearest player pawn to test the projectile's impact on the player.
  (Networking|SwitchHasAuthority
    (:Authority
      (if bEnable
        (bind _outfortplayerpawns (Fortnite|GetFortPlayerPawns))
        (for _array_element _outfortplayerpawns
          (Utilities|IsValid _array_element
            (:Is%20Valid
              (bind _playerstate (Variables|/Script/Engine.Pawn|GetPlayerState _array_element)
                (:then
                  (bind _asfort_player_state_athena (Utilities|Casting|CastToFortPlayerStateAthena _playerstate))
                  (if (and (Class|FortPlayerStateAthena|IsAnAthenaGameParticipant _asfort_player_state_athena) (.bSuccess (Utilities|Casting|CastToFortPlayerControllerAthena (Pawn|GetController _array_element))))
                    (Utilities|Array|Add (LocalVariables|GetHumanPlayerPawns) _array_element)))
                (:else)))))
        (bind _returnvalue (Fortnite|GetClosestActorfromArray self (LocalVariables|GetHumanPlayerPawns)))
        (Utilities|IsValid _returnvalue
          (:Is%20Valid
            (Components|Activation|Deactivate (Variables|Default|GetFortNearbyActorsPerception))
            (Default|SetCurrentTarget _returnvalue)))
        (else
          (CallFunction|StartPerception))))))

(fn EditorDebugTick
  (Development|PrintString (Utilities|String|Append "TidalCraneToyFoxBoss: Frustration: " (Utilities|String|ToString(Float) (Variables|Transient|GetFrustration))) true false "(R=1.000000,G=0.134839,B=0.788774,A=1.000000)" 5.0 "TidalCraneToyFoxBossFrustration"))

(fn GetAbilitySystemComponent
  (return (Ability|GetFortAbilitySystemComponent self)))

(fn GetAnimData
  ;[0.05783 0.01407 0.15 0.5] GetAnimData: If stunned, look straight ahead and not at the target (parent behavior)
  (if (GameplayTags|HasMatchingGameplayTag "(TagName=\"Ability.TIdalCrane.ToyFox.Stunned.Active\")")
    (return (Transformation|GetActorTransform) (+ (Transformation|GetActorLocation) (* (Transformation|GetActorForwardVector) 5e+03)))
    (else
      (bind (_currentactortransform _lookatdirection) (|CallParentFunction))
      (return _currentactortransform _lookatdirection))))

(fn GetFrustration
  ;[0.0226 0.09042 0.15 0.5] Returns the value of frustration.
  (return (Variables|Transient|GetFrustration)))

(fn GetTargetProximityDuration
  ;[0.15 0.02244 0.1262 0.5] Returns the promixity duration to target giant pawn.
  (return (Variables|Transient|GetTargetProximityDuration)))

(fn IsLocalPlayerPiloting
  (for _array_element (Fortnite|GetLocalFortPlayerControllers true)
    (bind _asfort_player_pawn (Utilities|Casting|CastToFortPlayerPawn (Pawn|GetControlledPawn _array_element)))
    (if (GameplayTags|HasMatchingGameplayTag _asfort_player_pawn "(TagName=\"ToyFox.Player.Pilot.Primary\")")
      (return true)))
  (return false))

(fn M_StaggerDenied
  (Rendering|Material|SetCustomPrimitiveDataFloat (Variables|Default|GetSkeletalMeshComponentBudgeted) 33 (Utilities|Time|GetGameTimeinSeconds)))

(fn M_StaggerImmunityStartupValues
  (Rendering|Material|SetCustomPrimitiveDataFloat (Variables|Default|GetSkeletalMeshComponentBudgeted) 33 (Utilities|Time|GetGameTimeinSeconds)))

(fn OnRep_bCollisionEnabled
  (Collision|SetActorEnableCollision (Variables|Transient|GetbCollisionEnabled)))

(fn ResetFrustration
  ;[0.04274 0.07462 0.15 0.5] Resetting frustration.
  (Networking|SwitchHasAuthority
    (:Authority
      (Variables|Transient|SetFrustration 0.0))))

(fn ResetTargetProximityDuration
  ;[0.03317 0.115 0.1458 0.5] Reset the target proximity duration.
  (Networking|SwitchHasAuthority
    (:Authority
      (Variables|Transient|SetTargetProximityDuration 0.0)
      (Fortnite|FortShippingLog self "ResetTargetProximityDuration: Target proximity duration reset."))))

(fn SetBossCollisionEnable
  (params (bEnable :type Boolean))
  ;[0.032 0.15 0 0.5] SetBossCollisionEnable: Enable or disable the collision for gameplay purpose as necessary.
  (Networking|SwitchHasAuthority
    (:Authority
      (bind _updated_variable (Utilities|FlowControl|DeltaGateWildcard (Variables|Transient|GetbCollisionEnabled) bEnable)
        (:Success
          (bind _output_get (Variables|Transient|SetbCollisionEnabled _updated_variable))
          (Fortnite|FortShippingLog self (Utilities|String|Append "SetBossCollisionEnable: Set to " (Utilities|String|ToString(Boolean) _output_get))))))))

(fn SetEnableDestructionComponent
  (params (bEnable :type Boolean))
  (bind _tidalcranetoyfoxbossdestructioncomponent (Variables|Default|GetTidalCraneToyFoxBossDestructionComponent))
  ;[0.007266 0.1422 0.15 0.5] SetEnableDestructionComponent: Start / Stop destruction traces.
  (Networking|SwitchHasAuthority
    (:Authority
      (Fortnite|FortShippingLog self (Utilities|String|Append "SetEnableDestructionComponent: Set to " (Utilities|String|ToString(Boolean) bEnable)))
      (if bEnable
        (Class|BGiantPawnDestructionComponent|StartSkeletonDestructionTraces _tidalcranetoyfoxbossdestructioncomponent)
        (else
          (Class|BGiantPawnDestructionComponent|StopSkeletonDestructionTraces _tidalcranetoyfoxbossdestructioncomponent))))))

(fn ShowFinalStrikeWidget
  (params (DoShow :type Boolean))
  (Rendering|SetVisibility (Variables|Default|GetFortWidget) DoShow false))

(fn UpdateTargetProximityDuration
  ;[0.15 0.05657 0.1307 0.5] UpdateTargetProximityDuration: Update the proximity duration timer.
  (Networking|SwitchHasAuthority
    (:Authority
      (bind _currenttargetactor (CallFunction|GetCurrentTarget))
      (Utilities|IsValid _currenttargetactor
        (:Is%20Valid
          (if (<= (Math|Vector|Distance2DSquared(Vector) (Transformation|GetActorLocation _currenttargetactor) (Transformation|GetActorLocation self)) (Math|Float|Square (ScalableFloat|GetCachedScalableFloatValue (Variables|Hotfix|GetHotfix_ProximityDistanceThreshold))))
            (bind _output_get (Variables|Transient|SetTargetProximityDuration (+ (Variables|Transient|GetTargetProximityDuration) (Variables|Settings|GetTargetProximityUpdateFrequency))))
            (else
              (TargetProximity|ResetTargetProximityDuration)))
          (Development|PrintString (Utilities|String|Append "TidalCraneToyFoxBoss: ProximityDuration: " (Utilities|String|ToString(Float) _output_get)) true false "(R=1.000000,G=0.134839,B=0.788774,A=1.000000)" 5.0 "TidalCraneToyFoxBossProximityDuration"))
        (:Is%20Not%20Valid
          _)))))

(fn ConstructionScript
  (|CallParentFunction))
