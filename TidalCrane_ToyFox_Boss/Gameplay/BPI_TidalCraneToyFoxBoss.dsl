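; Blueprint interface for the TidalCrane ToyFox boss, declaring calls related to the ToyFox Queen.
; It exposes a getter for an Ability System Component, plus parameterless calls to apply a stun and start a takeover on the Queen.
; It holds only signatures, so the behavior behind each call lives in the implementing classes.
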(fn GetAbilitySystemComponent
  (return 0))

(fn ToyFoxQueen_ApplyStun)

(fn ToyFoxQueen_StartTakeOver)
