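; Destruction component for the TidalCrane ToyFox boss, supplying the skeletal mesh that the destruction logic updates.
; GetSkeletalMeshComponentToUpdate casts the owning actor to BGA_TidalCrane_Boss_ToyFox and returns the budgeted skeletal mesh inherited from BGA_TidalCrane_Boss, or null if the cast fails.
; BeginPlay and Tick hold no active logic, so any destruction behavior lives outside this Blueprint.
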(event EventBeginPlay
  (|CallParentFunction) ; This node is disabled and will not be called.\nDrag off pins to build functionality.
)

(event EventTick
  (params (DeltaSeconds :type Float%20%28single-precision%29)))

(fn GetSkeletalMeshComponentToUpdate
  ;[0.09824 0.02452 0.15 0.5] Return the skeletal mesh of the expected owner actor.
  (bind _asbga_tidal_crane_boss_toy_fox (Utilities|Casting|CastToBGA_TidalCrane_Boss_ToyFox (Components|GetOwner))
    (:then
      (return (Variables|/TidalCrane_Swarm_Boss/Gameplay/BGA_TidalCrane_Boss.BGA_TidalCrane_Boss_C|GetSkeletalMeshComponentBudgeted _asbga_tidal_crane_boss_toy_fox)))
    (:CastFailed
      (return 0))))
