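# StateTree that drives the Toy Fox boss AI in TidalCrane, split into Combat and Roam branches, both gated by a FightGiant Blueprint condition.
# Combat's SelectAbility tries Reposition, MeleeAttack, ChangeTarget, then RangeAttack; Roam's tries Roam Reposition, then Roam Projectile Volley.
# Each ability state requires its tagged gameplay ability be activatable, starts it, then moves to that branch's Wait on completion.
# Wait transitions to Root on tick, re-running selection.
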StateTree /TidalCrane_ToyFox_Boss/Gameplay/ST_TidalCrane_Boss_ToyFox
states [
 Root {
  id=#1
  type=STATE
  children [#2, #3]
  tasks []
  transitions []
 }
 Combat {
  id=#2
  type=STATE
  children [#4, #5]
  enter_conditions [{
    id=#6
    description=B Tidal Crane Toy Fox Boss State Tree Condition Fight Giant
    properties {
     iD=#7
     instanceObject=&self:StateTreeEditorData_0.StateTreeState_1.B_TidalCraneToyFoxBossStateTreeCondition_GiantAlive_C_0
     node {
      name=None
      _structType=/Script/StateTreeModule.StateTreeBlueprintConditionWrapper
      conditionClass=/TidalCrane_ToyFox_Boss/Gameplay/StateTree/B_TidalCraneToyFoxBossStateTreeCondition_FightGiant.B_TidalCraneToyFoxBossStateTreeCondition_FightGiant_C
      deltaIndent=0
      evaluationMode=Evaluated
      operand=And
      bindingsBatch {value=65535}
      executionRuntimeTemplateIndex {value=65535}
      instanceDataHandle {index=65535, source=None, stateHandle {index=65535}}
      instanceTemplateIndex {value=65535}
      outputBindingsBatch {value=65535}
     }
    }
   }]
  tasks [{
    id=#8
    description="Debug Text \"COMBAT\""
    properties {
     iD=#9
     instance {_structType=/Script/StateTreeModule.StateTreeDebugTextTaskInstanceData, bEnabled=false, bindableText="", referenceActor=null}
     node {
      name=None
      _structType=/Script/StateTreeModule.StateTreeDebugTextTask
      bCanEditConsideredForCompletion=false
      bConsideredForCompletion=false
      bHasTaskCompletionDelegateDispatcher=false
      bTaskEnabled=true
      deprecatedVersion=1
      fontScale=1
      text=COMBAT
      transitionHandlingPriority=Normal
      bindingsBatch {value=65535}
      executionRuntimeTemplateIndex {value=65535}
      instanceDataHandle {index=65535, source=None, stateHandle {index=65535}}
      instanceTemplateIndex {value=65535}
      offset=<0 0 0>
      outputBindingsBatch {value=65535}
      textColor=<rgba 1 0 0.9765 1>
     }
    }
   }]
  transitions []
 }
 Roam {
  id=#3
  type=STATE
  children [#10, #11]
  enter_conditions [{
    id=#12
    description=B Tidal Crane Toy Fox Boss State Tree Condition Fight Giant
    properties {
     iD=#13
     instanceObject=&self:StateTreeEditorData_0.StateTreeState_2.B_TidalCraneToyFoxBossStateTreeCondition_GiantAlive_C_0
     node {
      name=None
      _structType=&StateTreeBlueprintConditionWrapper
      conditionClass=&B_TidalCraneToyFoxBossStateTreeCondition_FightGiant_C
      deltaIndent=0
      evaluationMode=Evaluated
      operand=And
      bindingsBatch {value=65535}
      executionRuntimeTemplateIndex {value=65535}
      instanceDataHandle {index=65535, source=None, stateHandle {index=65535}}
      instanceTemplateIndex {value=65535}
      outputBindingsBatch {value=65535}
     }
    }
   }]
  tasks []
  transitions []
 }
 SelectAbility {
  id=#4
  type=STATE
  children [#14, #15, #16, #17]
  tasks []
  transitions []
 }
 Wait {
  id=#5
  type=STATE
  tasks [{
    id=#18
    description=Delay Forever
    properties {
     iD=#19
     instance {_structType=&StateTreeDelayTaskInstanceData, bRunForever=true, duration=1}
     node {name=None, _structType=&StateTreeDelayTask}
    }
   }]
  transitions [{id=#20, trigger=ON_TICK, state {id=#1, name=Root}}]
 }
 SelectAbility {
  id=#10
  type=STATE
  children [#21, #22]
  tasks []
  transitions []
 }
 Wait {
  id=#11
  type=STATE
  tasks [{
    id=#23
    description=Delay Forever
    properties {
     iD=#24
     instance {_structType=&StateTreeDelayTaskInstanceData, bRunForever=true, duration=1}
     node {name=None, _structType=&StateTreeDelayTask}
    }
   }]
  transitions [{id=#25, trigger=ON_TICK, state {id=#1, name=Root}}]
 }
 Reposition {
  id=#14
  type=STATE
  enter_conditions [{
    id=#26
    description=Can Activate Ability Ability.TidalCrane.Boss.Reposition
    properties {
     iD=#27
     instance {
      _structType=/Script/FortniteAIServer.FortGameplayAbilityCanActivateAbilityInstanceData
      pawn=null
      abilityTag {gameplayTags [{tagName=Ability.TidalCrane.Boss.Reposition}]}
     }
     node {
      name=None
      _structType=/Script/FortniteAIServer.FortGameplayAbilityCanActivateAbilityCondition
      bInvert=false
      deltaIndent=0
      evaluationMode=Evaluated
      operand=And
      bindingsBatch {value=65535}
      executionRuntimeTemplateIndex {value=65535}
      instanceDataHandle {index=65535, source=None, stateHandle {index=65535}}
      instanceTemplateIndex {value=65535}
      outputBindingsBatch {value=65535}
     }
    }
   }]
  tasks [{
    id=#28
    description=Start Ability Ability.TidalCrane.Boss.Reposition
    properties {
     iD=#29
     instance {
      _structType=/Script/FortniteGame.FortStateTreeStartAbilityTaskInstanceData
      actor=null
      abilityTag {gameplayTags [{tagName=Ability.TidalCrane.Boss.Reposition}]}
     }
     node {
      name=None
      _structType=/Script/FortniteGame.FortStateTreeStartAbilityTask
      bCanEditConsideredForCompletion=true
      bConsideredForCompletion=true
      bHasTaskCompletionDelegateDispatcher=false
      bTaskEnabled=true
      transitionHandlingPriority=Normal
      bindingsBatch {value=65535}
      executionRuntimeTemplateIndex {value=65535}
      instanceDataHandle {index=65535, source=None, stateHandle {index=65535}}
      instanceTemplateIndex {value=65535}
      outputBindingsBatch {value=65535}
     }
    }
   }]
  transitions [{id=#30, trigger=ON_STATE_COMPLETED, state {id=#5, name=Wait}}]
 }
 MeleeAttack {
  id=#15
  type=STATE
  enter_conditions [{
    id=#31
    description=Can Activate Ability Ability.TidalCrane.Boss.Melee
    properties {
     iD=#32
     instance {
      _structType=&FortGameplayAbilityCanActivateAbilityInstanceData
      pawn=null
      abilityTag {gameplayTags [{tagName=Ability.TidalCrane.Boss.Melee}]}
     }
     node {
      name=None
      _structType=&FortGameplayAbilityCanActivateAbilityCondition
      bInvert=false
      deltaIndent=0
      evaluationMode=Evaluated
      operand=And
      bindingsBatch {value=65535}
      executionRuntimeTemplateIndex {value=65535}
      instanceDataHandle {index=65535, source=None, stateHandle {index=65535}}
      instanceTemplateIndex {value=65535}
      outputBindingsBatch {value=65535}
     }
    }
   }]
  tasks [{
    id=#33
    description=Start Ability Ability.TidalCrane.Boss.Melee
    properties {
     iD=#34
     instance {
      _structType=&FortStateTreeStartAbilityTaskInstanceData
      actor=null
      abilityTag {gameplayTags [{tagName=Ability.TidalCrane.Boss.Melee}]}
     }
     node {
      name=None
      _structType=&FortStateTreeStartAbilityTask
      bCanEditConsideredForCompletion=true
      bConsideredForCompletion=true
      bHasTaskCompletionDelegateDispatcher=false
      bTaskEnabled=true
      transitionHandlingPriority=Normal
      bindingsBatch {value=65535}
      executionRuntimeTemplateIndex {value=65535}
      instanceDataHandle {index=65535, source=None, stateHandle {index=65535}}
      instanceTemplateIndex {value=65535}
      outputBindingsBatch {value=65535}
     }
    }
   }]
  transitions [{id=#35, trigger=ON_STATE_COMPLETED, state {id=#5, name=Wait}}]
 }
 ChangeTarget {
  id=#16
  type=STATE
  enter_conditions [{
    id=#36
    description=Can Activate Ability Ability.TidalCrane.Boss.ChangeTarget
    properties {
     iD=#37
     instance {
      _structType=&FortGameplayAbilityCanActivateAbilityInstanceData
      pawn=null
      abilityTag {gameplayTags [{tagName=Ability.TidalCrane.Boss.ChangeTarget}]}
     }
     node {
      name=None
      _structType=&FortGameplayAbilityCanActivateAbilityCondition
      bInvert=false
      deltaIndent=0
      evaluationMode=Evaluated
      operand=And
      bindingsBatch {value=65535}
      executionRuntimeTemplateIndex {value=65535}
      instanceDataHandle {index=65535, source=None, stateHandle {index=65535}}
      instanceTemplateIndex {value=65535}
      outputBindingsBatch {value=65535}
     }
    }
   }]
  tasks [{
    id=#38
    description=Start Ability Ability.TidalCrane.Boss.ChangeTarget
    properties {
     iD=#39
     instance {
      _structType=&FortStateTreeStartAbilityTaskInstanceData
      actor=null
      abilityTag {gameplayTags [{tagName=Ability.TidalCrane.Boss.ChangeTarget}]}
     }
     node {
      name=None
      _structType=&FortStateTreeStartAbilityTask
      bCanEditConsideredForCompletion=true
      bConsideredForCompletion=true
      bHasTaskCompletionDelegateDispatcher=false
      bTaskEnabled=true
      transitionHandlingPriority=Normal
      bindingsBatch {value=65535}
      executionRuntimeTemplateIndex {value=65535}
      instanceDataHandle {index=65535, source=None, stateHandle {index=65535}}
      instanceTemplateIndex {value=65535}
      outputBindingsBatch {value=65535}
     }
    }
   }]
  transitions [{id=#40, trigger=ON_STATE_COMPLETED, state {id=#5, name=Wait}}]
 }
 RangeAttack {
  id=#17
  type=STATE
  enter_conditions [{
    id=#41
    description=Can Activate Ability Ability.TidalCrane.Boss.RangeAttack
    properties {
     iD=#42
     instance {
      _structType=&FortGameplayAbilityCanActivateAbilityInstanceData
      pawn=null
      abilityTag {gameplayTags [{tagName=Ability.TidalCrane.Boss.RangeAttack}]}
     }
     node {
      name=None
      _structType=&FortGameplayAbilityCanActivateAbilityCondition
      bInvert=false
      deltaIndent=0
      evaluationMode=Evaluated
      operand=And
      bindingsBatch {value=65535}
      executionRuntimeTemplateIndex {value=65535}
      instanceDataHandle {index=65535, source=None, stateHandle {index=65535}}
      instanceTemplateIndex {value=65535}
      outputBindingsBatch {value=65535}
     }
    }
   }]
  tasks [{
    id=#43
    description=Start Ability Ability.TidalCrane.Boss.RangeAttack
    properties {
     iD=#44
     instance {
      _structType=&FortStateTreeStartAbilityTaskInstanceData
      actor=null
      abilityTag {gameplayTags [{tagName=Ability.TidalCrane.Boss.RangeAttack}]}
     }
     node {
      name=None
      _structType=&FortStateTreeStartAbilityTask
      bCanEditConsideredForCompletion=true
      bConsideredForCompletion=true
      bHasTaskCompletionDelegateDispatcher=false
      bTaskEnabled=true
      transitionHandlingPriority=Normal
      bindingsBatch {value=65535}
      executionRuntimeTemplateIndex {value=65535}
      instanceDataHandle {index=65535, source=None, stateHandle {index=65535}}
      instanceTemplateIndex {value=65535}
      outputBindingsBatch {value=65535}
     }
    }
   }]
  transitions [{id=#45, trigger=ON_STATE_COMPLETED, state {id=#5, name=Wait}}]
 }
 Roam Reposition {
  id=#21
  type=STATE
  enter_conditions [{
    id=#46
    description=Can Activate Ability Ability.TidalCrane.Boss.Roam.Reposition
    properties {
     iD=#47
     instance {
      _structType=&FortGameplayAbilityCanActivateAbilityInstanceData
      pawn=null
      abilityTag {gameplayTags [{tagName=Ability.TidalCrane.Boss.Roam.Reposition}]}
     }
     node {
      name=None
      _structType=&FortGameplayAbilityCanActivateAbilityCondition
      bInvert=false
      deltaIndent=0
      evaluationMode=Evaluated
      operand=And
      bindingsBatch {value=65535}
      executionRuntimeTemplateIndex {value=65535}
      instanceDataHandle {index=65535, source=None, stateHandle {index=65535}}
      instanceTemplateIndex {value=65535}
      outputBindingsBatch {value=65535}
     }
    }
   }]
  tasks [{
    id=#48
    description=Start Ability Ability.TidalCrane.Boss.Roam.Reposition
    properties {
     iD=#49
     instance {
      _structType=&FortStateTreeStartAbilityTaskInstanceData
      actor=null
      abilityTag {gameplayTags [{tagName=Ability.TidalCrane.Boss.Roam.Reposition}]}
     }
     node {
      name=None
      _structType=&FortStateTreeStartAbilityTask
      bCanEditConsideredForCompletion=true
      bConsideredForCompletion=true
      bHasTaskCompletionDelegateDispatcher=false
      bTaskEnabled=true
      transitionHandlingPriority=Normal
      bindingsBatch {value=65535}
      executionRuntimeTemplateIndex {value=65535}
      instanceDataHandle {index=65535, source=None, stateHandle {index=65535}}
      instanceTemplateIndex {value=65535}
      outputBindingsBatch {value=65535}
     }
    }
   }]
  transitions [{id=#50, trigger=ON_STATE_COMPLETED, state {id=#11, name=Wait}}]
 }
 Roam Projectile Volley {
  id=#22
  type=STATE
  enter_conditions [{
    id=#51
    description=Can Activate Ability TidalCrane.ToyFox.Boss.Ability.Roam.ProjectileVolley
    properties {
     iD=#52
     instance {
      _structType=&FortGameplayAbilityCanActivateAbilityInstanceData
      pawn=null
      abilityTag {gameplayTags [{tagName=TidalCrane.ToyFox.Boss.Ability.Roam.ProjectileVolley}]}
     }
     node {
      name=None
      _structType=&FortGameplayAbilityCanActivateAbilityCondition
      bInvert=false
      deltaIndent=0
      evaluationMode=Evaluated
      operand=And
      bindingsBatch {value=65535}
      executionRuntimeTemplateIndex {value=65535}
      instanceDataHandle {index=65535, source=None, stateHandle {index=65535}}
      instanceTemplateIndex {value=65535}
      outputBindingsBatch {value=65535}
     }
    }
   }]
  tasks [{
    id=#53
    description=Start Ability TidalCrane.ToyFox.Boss.Ability.Roam.ProjectileVolley
    properties {
     iD=#54
     instance {
      _structType=&FortStateTreeStartAbilityTaskInstanceData
      actor=null
      abilityTag {gameplayTags [{tagName=TidalCrane.ToyFox.Boss.Ability.Roam.ProjectileVolley}]}
     }
     node {
      name=None
      _structType=&FortStateTreeStartAbilityTask
      bCanEditConsideredForCompletion=true
      bConsideredForCompletion=true
      bHasTaskCompletionDelegateDispatcher=false
      bTaskEnabled=true
      transitionHandlingPriority=Normal
      bindingsBatch {value=65535}
      executionRuntimeTemplateIndex {value=65535}
      instanceDataHandle {index=65535, source=None, stateHandle {index=65535}}
      instanceTemplateIndex {value=65535}
      outputBindingsBatch {value=65535}
     }
    }
   }]
  transitions [{id=#55, trigger=ON_STATE_COMPLETED, state {id=#11, name=Wait}}]
 }
]
